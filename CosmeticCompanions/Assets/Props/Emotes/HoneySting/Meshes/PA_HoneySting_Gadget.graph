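# Physics asset for the HoneySting gadget, an emote prop in the CosmeticCompanions content.
# A single body on the main bone uses one short, wide, nearly spherical capsule, offset and rotated from the bone origin, as a rough volume for the prop.
# It has no constraints, so it defines no jointed or multi-body ragdoll setup.
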PhysicsAsset /CosmeticCompanions/Assets/Props/Emotes/HoneySting/Meshes/PA_HoneySting_Gadget
bodies {main [{shapeType=Capsule, length=14.7222, radius=35.5602, center=<0.7735 0.3361 18.6496>, rotation=<rot 0.8718 -90 9.8516>}]}
constraints []
